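# Physics asset for the dynamic back skirt of the female medium Swashbuckler player body, Frontend variant.
# Four box-body tail chains of differing length hang from spine_02, with tight twist limits and narrower swing at the roots.
# Cross-constraints tie neighboring chains at select segments so the panels behave like one sheet.
# Capsules and spheres on pelvis, spine, legs and right foot form body colliders, and a root box forms a ground slab.
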PhysicsAsset /Game/Characters/Player/Female/Medium/Bodies/F_MED_Swashbuckler/Meshes/F_MED_Swashbuckler_Dyn_Skirt_Back_Frontend
bodies {
 calf_l [
  {shapeType=Capsule, length=0, radius=8.9189, center=<-18.4402 -4.315 -0.6383>, rotation=<rot 90 0 0>}
  {shapeType=Capsule, length=25.7, radius=4.4296, center=<-28.4402 -4.3151 -0.4886>, rotation=<rot 90 53.9726 63.9726>}
  {shapeType=Capsule, length=6.9149, radius=5.8622, center=<-8.8682 -3.5457 -0.0057>, rotation=<rot 84.9974 -164.9984 179.9999>}
 ]
 calf_r [
  {shapeType=Capsule, length=0, radius=8.4189, center=<18.4402 4.315 0.6383>, rotation=<rot 90 0 0>}
  {shapeType=Capsule, length=25.7, radius=4.4296, center=<28.4402 4.3151 0.4886>, rotation=<rot 90 53.9726 63.9726>}
  {shapeType=Capsule, length=6.9149, radius=5.8622, center=<8.8682 3.5457 0.0057>, rotation=<rot 84.9974 -164.9984 179.9999>}
 ]
 dyn_tail_l_01 [{shapeType=Box, extentX=10, extentY=11.5, extentZ=1.1, center=<5.7814 -3.5175 -0.0002>}]
 dyn_tail_l_02 [{shapeType=Box, extentX=10, extentY=11.5, extentZ=1.1, center=<5.7814 -3.311 0.0002>}]
 dyn_tail_l_03 [{shapeType=Box, extentX=12.9758, extentY=11.5, extentZ=1.1, center=<6.1935 -1.874 -0.0002>}]
 dyn_tail_l_04 [{shapeType=Box, extentX=11.5529, extentY=11.5, extentZ=1.1, center=<5.3761 -1.3282 0>}]
 dyn_tail_mid_l_01 [{shapeType=Box, extentX=10, extentY=11.5, extentZ=1.1, center=<5.7814 0 0>}]
 dyn_tail_mid_l_02 [{shapeType=Box, extentX=10, extentY=11.5, extentZ=1.1, center=<5.7814 0 0>}]
 dyn_tail_mid_l_03 [{shapeType=Box, extentX=12.9399, extentY=11.5, extentZ=1.1, center=<5.9879 0 0>}]
 dyn_tail_mid_l_04 [{shapeType=Box, extentX=13.3593, extentY=11.5, extentZ=1.1, center=<5.9945 0 0>}]
 dyn_tail_mid_l_05 [{shapeType=Box, extentX=8.8752, extentY=11.5, extentZ=1.1, center=<3.6999 0 -0.0001>}]
 dyn_tail_mid_r_01 [{shapeType=Box, extentX=10, extentY=11.5, extentZ=1.1, center=<5.7814 0 0>}]
 dyn_tail_mid_r_02 [{shapeType=Box, extentX=10, extentY=11.5, extentZ=1.1, center=<5.7814 0 0>}]
 dyn_tail_mid_r_03 [{shapeType=Box, extentX=13.1154, extentY=11.5, extentZ=1.1, center=<6.419 0 0>}]
 dyn_tail_mid_r_04 [{shapeType=Box, extentX=11.8001, extentY=14.2574, extentZ=1.1, center=<5.7814 0 0>}]
 dyn_tail_mid_r_05 [{shapeType=Box, extentX=10, extentY=14.2574, extentZ=1.1, center=<5.7814 0 0>}]
 dyn_tail_r_01 [{shapeType=Box, extentX=10, extentY=11.5, extentZ=1.1, center=<5.7814 0 0>}]
 dyn_tail_r_02 [{shapeType=Box, extentX=12.2492, extentY=11.5, extentZ=1.1, center=<5.7814 0 0>}]
 dyn_tail_r_03 [{shapeType=Box, extentX=13.2431, extentY=11.5, extentZ=1.1, center=<5.7814 0 0>}]
 dyn_tail_r_04 [{shapeType=Box, extentX=11.4558, extentY=11.5, extentZ=1.1, center=<4.9383 0 0>}]
 dyn_tail_r_05 [{shapeType=Box, extentX=12.7705, extentY=11.5, extentZ=1.1, center=<6.1805 0 0.0005>}]
 dyn_tail_r_06 [{shapeType=Box, extentX=13.9202, extentY=11.5, extentZ=1.1, center=<6.7657 0 0.0001>}]
 dyn_tail_r_07 [{shapeType=Box, extentX=12.2802, extentY=11.5, extentZ=1.1, center=<5.7814 0 0>}]
 foot_r [{shapeType=Capsule, length=8.0069, radius=6.8858, center=<2.8171 -2.8116 -0.167>, rotation=<rot 0 18.4005 -86.2651>}]
 pelvis [
  {shapeType=Capsule, length=8.0799, radius=10.5, center=<0 -0.7015 0.0001>}
  {shapeType=Capsule, length=11.0884, radius=10.5, center=<-6.8811 -1.981 -0.0001>}
 ]
 root [{shapeType=Box, extentX=100, extentY=100, extentZ=10, center=<0 0 -5>}]
 spine_01 [{shapeType=Capsule, length=9.5934, radius=9.1043, center=<0.0001 1.4126 0>}]
 spine_02 [{shapeType=Capsule, length=0, radius=6.1053}]
 thigh_l [
  {shapeType=Sphere, radius=9, center=<-4.8394 -2.9083 2.4184>}
  {shapeType=Capsule, length=34.5588, radius=7.2188, center=<-25.5621 -0.0989 0.0174>, rotation=<rot 87.4256 176.5118 177.8422>}
  {shapeType=Capsule, length=19.8509, radius=3.116, center=<-7.8956 -0.7218 -9.8719>, rotation=<rot 84.9993 -174.9999 180>}
 ]
 thigh_r [
  {shapeType=Sphere, radius=9, center=<4.8394 2.9083 -2.4184>}
  {shapeType=Capsule, length=34.5588, radius=7.2188, center=<25.5621 0.0989 -0.0174>, rotation=<rot 87.4256 176.5118 177.8422>}
 ]
}
constraints [
 {bone1Name=dyn_tail_r_02, bone2Name=dyn_tail_r_01, swing1LimitDegrees=45, swing2LimitDegrees=45, twistLimitDegrees=2}
 {bone1Name=dyn_tail_r_03, bone2Name=dyn_tail_r_02, swing1LimitDegrees=45, swing2LimitDegrees=45, twistLimitDegrees=2}
 {bone1Name=dyn_tail_r_04, bone2Name=dyn_tail_r_03, swing1LimitDegrees=45, swing2LimitDegrees=45, twistLimitDegrees=2}
 {bone1Name=dyn_tail_r_05, bone2Name=dyn_tail_r_04, swing1LimitDegrees=45, swing2LimitDegrees=45, twistLimitDegrees=2}
 {bone1Name=dyn_tail_r_06, bone2Name=dyn_tail_r_05, swing1LimitDegrees=45, swing2LimitDegrees=45, twistLimitDegrees=2}
 {bone1Name=dyn_tail_r_07, bone2Name=dyn_tail_r_06, swing1LimitDegrees=45, swing2LimitDegrees=45, twistLimitDegrees=2}
 {bone1Name=dyn_tail_mid_r_02, bone2Name=dyn_tail_mid_r_01, swing1LimitDegrees=45, swing2LimitDegrees=45, twistLimitDegrees=2}
 {bone1Name=dyn_tail_mid_r_03, bone2Name=dyn_tail_mid_r_02, swing1LimitDegrees=45, swing2LimitDegrees=45, twistLimitDegrees=2}
 {bone1Name=dyn_tail_mid_r_04, bone2Name=dyn_tail_mid_r_03, swing1LimitDegrees=45, swing2LimitDegrees=45, twistLimitDegrees=2}
 {bone1Name=dyn_tail_mid_r_05, bone2Name=dyn_tail_mid_r_04, swing1LimitDegrees=45, swing2LimitDegrees=45, twistLimitDegrees=2}
 {bone1Name=dyn_tail_mid_l_02, bone2Name=dyn_tail_mid_l_01, swing1LimitDegrees=45, swing2LimitDegrees=45, twistLimitDegrees=2}
 {bone1Name=dyn_tail_mid_l_03, bone2Name=dyn_tail_mid_l_02, swing1LimitDegrees=45, swing2LimitDegrees=45, twistLimitDegrees=2}
 {bone1Name=dyn_tail_mid_l_04, bone2Name=dyn_tail_mid_l_03, swing1LimitDegrees=45, swing2LimitDegrees=45, twistLimitDegrees=2}
 {bone1Name=dyn_tail_mid_l_05, bone2Name=dyn_tail_mid_l_04, swing1LimitDegrees=45, swing2LimitDegrees=45, twistLimitDegrees=2}
 {bone1Name=dyn_tail_l_02, bone2Name=dyn_tail_l_01, swing1LimitDegrees=45, swing2LimitDegrees=45, twistLimitDegrees=2}
 {bone1Name=dyn_tail_l_03, bone2Name=dyn_tail_l_02, swing1LimitDegrees=45, swing2LimitDegrees=45, twistLimitDegrees=2}
 {bone1Name=dyn_tail_l_04, bone2Name=dyn_tail_l_03, swing1LimitDegrees=45, swing2LimitDegrees=45, twistLimitDegrees=2}
 {bone1Name=dyn_tail_r_01, bone2Name=spine_02, swing1LimitDegrees=10, swing2LimitDegrees=45, twistLimitDegrees=2}
 {bone1Name=dyn_tail_mid_r_01, bone2Name=spine_02, swing1LimitDegrees=10, swing2LimitDegrees=45, twistLimitDegrees=2}
 {bone1Name=dyn_tail_mid_l_01, bone2Name=spine_02, swing1LimitDegrees=10, swing2LimitDegrees=45, twistLimitDegrees=2}
 {bone1Name=dyn_tail_l_01, bone2Name=spine_02, swing1LimitDegrees=10, swing2LimitDegrees=45, twistLimitDegrees=2}
 {bone1Name=spine_02, bone2Name=spine_01, swing1LimitDegrees=45, swing2LimitDegrees=45, twistLimitDegrees=2}
 {bone1Name=thigh_l, bone2Name=pelvis, swing1LimitDegrees=45, swing2LimitDegrees=45, twistLimitDegrees=2}
 {bone1Name=thigh_r, bone2Name=pelvis, swing1LimitDegrees=45, swing2LimitDegrees=45, twistLimitDegrees=2}
 {
  bone1Name=calf_l
  bone2Name=thigh_l
  swing1LimitDegrees=45
  swing1Motion=Free
  swing2LimitDegrees=45
  swing2Motion=Free
  twistLimitDegrees=45
  twistMotion=Free
 }
 {
  bone1Name=calf_r
  bone2Name=thigh_r
  swing1LimitDegrees=45
  swing1Motion=Free
  swing2LimitDegrees=45
  swing2Motion=Free
  twistLimitDegrees=45
  twistMotion=Free
 }
 {
  bone1Name=spine_01
  bone2Name=pelvis
  swing1LimitDegrees=45
  swing1Motion=Free
  swing2LimitDegrees=45
  swing2Motion=Free
  twistLimitDegrees=45
  twistMotion=Free
 }
 {bone1Name=dyn_tail_r_02, bone2Name=dyn_tail_mid_r_02, swing1LimitDegrees=45, swing2LimitDegrees=45, twistLimitDegrees=45}
 {bone1Name=dyn_tail_r_04, bone2Name=dyn_tail_mid_r_04, swing1LimitDegrees=45, swing2LimitDegrees=45, twistLimitDegrees=45}
 {bone1Name=dyn_tail_l_02, bone2Name=dyn_tail_mid_l_02, swing1LimitDegrees=45, swing2LimitDegrees=45, twistLimitDegrees=45}
 {
  bone1Name=dyn_tail_l_04
  bone2Name=dyn_tail_mid_l_04
  swing1LimitDegrees=45
  swing1Motion=Free
  swing2LimitDegrees=45
  swing2Motion=Free
  twistLimitDegrees=45
  twistMotion=Free
 }
 {bone1Name=dyn_tail_mid_r_03, bone2Name=dyn_tail_mid_l_03, swing1LimitDegrees=45, swing2LimitDegrees=45, twistLimitDegrees=45}
 {
  bone1Name=pelvis
  bone2Name=root
  swing1LimitDegrees=45
  swing1Motion=Free
  swing2LimitDegrees=45
  swing2Motion=Free
  twistLimitDegrees=45
  twistMotion=Free
 }
 {
  bone1Name=foot_r
  bone2Name=calf_r
  swing1LimitDegrees=45
  swing1Motion=Free
  swing2LimitDegrees=45
  swing2Motion=Free
  twistLimitDegrees=45
  twistMotion=Free
 }
 {
  bone1Name=dyn_tail_mid_r_05
  bone2Name=dyn_tail_mid_l_05
  swing1LimitDegrees=45
  swing1Motion=Free
  swing2LimitDegrees=45
  swing2Motion=Free
  twistLimitDegrees=45
  twistMotion=Free
 }
]
