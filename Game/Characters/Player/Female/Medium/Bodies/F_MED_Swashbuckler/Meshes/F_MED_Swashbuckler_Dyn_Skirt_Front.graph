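# Physics asset for the front skirt (loincloth bones) of the medium female Swashbuckler player body.
# Left and right four-bone capsule chains have limited swing and tight twist, and both hang from the right thigh.
# A free-rotation constraint links the two chain tips.
# Capsules on pelvis and thighs plus a flat root box provide body and ground shapes, with thighs and the loincloth master bone constrained to the pelvis.
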PhysicsAsset /Game/Characters/Player/Female/Medium/Bodies/F_MED_Swashbuckler/Meshes/F_MED_Swashbuckler_Dyn_Skirt_Front
bodies {
 dyn_loincloth_01_l [{shapeType=Capsule, length=9.5, radius=5, shapeName=dyn_loincloth_01_l_capsule, center=<7.0583 0 4.2>, rotation=<rot 90 0 0>}]
 dyn_loincloth_01_r [{shapeType=Capsule, length=10.25, radius=5.75, shapeName=dyn_loincloth_01_r_capsule, center=<6.5303 0 5>, rotation=<rot 90 0 0>}]
 dyn_loincloth_02_l [{shapeType=Capsule, length=13.5, radius=4.75, shapeName=dyn_loincloth_02_l_capsule, center=<8.0329 0 3.8>, rotation=<rot 90 0 0>}]
 dyn_loincloth_02_r [{shapeType=Capsule, length=15.5, radius=6, shapeName=dyn_loincloth_02_r_capsule, center=<9.5014 0 5.3>, rotation=<rot 90 0 0>}]
 dyn_loincloth_03_l [{shapeType=Capsule, length=9.5, radius=3.5, shapeName=dyn_loincloth_03_l_capsule, center=<6.28 0 2.6>, rotation=<rot 90 0 0>}]
 dyn_loincloth_03_r [{shapeType=Capsule, length=12.75, radius=7.5, shapeName=dyn_loincloth_03_r_capsule, center=<7.6812 0 6.7>, rotation=<rot 90 0 0>}]
 dyn_loincloth_04_l [{shapeType=Capsule, length=0, radius=5, shapeName=dyn_loincloth_04_l_capsule, center=<2.1439 0 4.1>, rotation=<rot 90 0 0>}]
 dyn_loincloth_04_r [{shapeType=Capsule, length=4.75, radius=5.25, shapeName=dyn_loincloth_04_r_capsule, center=<4.0421 0 4.5>, rotation=<rot 90 0 0>}]
 dyn_loincloth_MSTR [{shapeType=Capsule, length=0.4885, radius=0.5501, rotation=<rot 90 0 0>}]
 pelvis [{shapeType=Capsule, length=7.25, radius=10.5, center=<-2.8439 -2.8254 0>}]
 root [{shapeType=Box, extentX=100, extentY=100, extentZ=10, center=<0 0 -5>}]
 thigh_l [{shapeType=Capsule, length=17.6389, radius=7.2373, center=<-23.9793 0.021 -0.997>, rotation=<rot 84.6012 177.1213 177.1701>}]
 thigh_r [
  {shapeType=Capsule, length=25.891, radius=7.3455, center=<37.0015 0.765 1.757>, rotation=<rot 84.6032 177.1213 171.1016>}
  {shapeType=Capsule, length=12.0965, radius=7.84, center=<16.937 -0.3925 0>, rotation=<rot 90 0 0>}
 ]
}
constraints [
 {
  bone1Name=thigh_l
  bone2Name=pelvis
  swing1LimitDegrees=45
  swing1Motion=Free
  swing2LimitDegrees=45
  swing2Motion=Free
  twistLimitDegrees=45
  twistMotion=Free
 }
 {
  bone1Name=thigh_r
  bone2Name=pelvis
  swing1LimitDegrees=45
  swing1Motion=Free
  swing2LimitDegrees=45
  swing2Motion=Free
  twistLimitDegrees=45
  twistMotion=Free
 }
 {
  bone1Name=dyn_loincloth_MSTR
  bone2Name=pelvis
  swing1LimitDegrees=45
  swing1Motion=Free
  swing2LimitDegrees=45
  swing2Motion=Free
  twistLimitDegrees=45
  twistMotion=Free
 }
 {bone1Name=dyn_loincloth_01_l, bone2Name=thigh_r, swing1LimitDegrees=30, swing2LimitDegrees=30, twistLimitDegrees=5}
 {bone1Name=dyn_loincloth_02_l, bone2Name=dyn_loincloth_01_l, swing1LimitDegrees=30, swing2LimitDegrees=30, twistLimitDegrees=5}
 {bone1Name=dyn_loincloth_03_l, bone2Name=dyn_loincloth_02_l, swing1LimitDegrees=30, swing2LimitDegrees=30, twistLimitDegrees=5}
 {bone1Name=dyn_loincloth_04_l, bone2Name=dyn_loincloth_03_l, swing1LimitDegrees=30, swing2LimitDegrees=30, twistLimitDegrees=5}
 {bone1Name=dyn_loincloth_01_r, bone2Name=thigh_r, swing1LimitDegrees=30, swing2LimitDegrees=30, twistLimitDegrees=5}
 {bone1Name=dyn_loincloth_02_r, bone2Name=dyn_loincloth_01_r, swing1LimitDegrees=30, swing2LimitDegrees=30, twistLimitDegrees=5}
 {bone1Name=dyn_loincloth_03_r, bone2Name=dyn_loincloth_02_r, swing1LimitDegrees=30, swing2LimitDegrees=30, twistLimitDegrees=5}
 {bone1Name=dyn_loincloth_04_r, bone2Name=dyn_loincloth_03_r, swing1LimitDegrees=30, swing2LimitDegrees=30, twistLimitDegrees=5}
 {
  bone1Name=pelvis
  bone2Name=root
  swing1LimitDegrees=45
  swing1Motion=Free
  swing2LimitDegrees=45
  swing2Motion=Free
  twistLimitDegrees=45
  twistMotion=Free
 }
 {
  bone1Name=dyn_loincloth_04_r
  bone2Name=dyn_loincloth_04_l
  swing1LimitDegrees=45
  swing1Motion=Free
  swing2LimitDegrees=45
  swing2Motion=Free
  twistLimitDegrees=45
  twistMotion=Free
 }
]
