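# Post-process style AnimBP for the female Swashbuckler body, copying the parent mesh pose and adding secondary motion to coat tails, wrist flaps and boot cuffs.
# RigidBody skirt sims and AnimDynamics run in two chains, with a Frontend back-skirt asset selected by bCanPlayCustomAnimations.
# Bone constraints replace simulation when rigid bodies are disabled or LOD exceeds 1.
# EventGraph varies skirt gravity for swim-sprinting, skydiving and parachuting; skydiving also pitches the tails.
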AnimBlueprint /Game/Characters/Player/Female/Medium/Bodies/F_MED_Swashbuckler/Meshes/F_MED_Swashbuckler_AnimBP
target_skeleton=&/F_MED_Swashbuckler_Skeleton
graphs [
 AnimGraph {nodes [
   n1 = "Constraint - thigh_r_blend (R)"(ComponentPose=Local To Component(LocalPose=Use cached pose 'copypose_cache'()), Alpha=0.5, ConstraintWeights_0=1) {
    alphaBoolBlend {blendOption=Linear}
    boneToModify {boneName=thigh_r_blend}
    constraintSetup [{transformType=Rotation, targetBone {boneName=thigh_r}}]
    constraintWeights [1]
   }
   n2 = "Constraint - thigh_l_blend (R)"(ComponentPose=@n1, Alpha=0.5, ConstraintWeights_0=1) {
    alphaBoolBlend {blendOption=Linear}
    boneToModify {boneName=thigh_l_blend}
    constraintSetup [{transformType=Rotation, targetBone {boneName=thigh_l}}]
    constraintWeights [1]
   }
   n3 = Break Fort Anim Input Skydiving External Force(FortAnimInput_SkydivingExternalForce=Get SkydivingExternalForceInput()) -> (
    bUseSkydivingVectorForce=~
    SkydivingVectorForce=~
    HeadToPelvisDirection=~
    FloatingMultiplier=~
    FloatingAdditive=~
    DivingMultiplier=~
    DivingAdditive=~
    ParachutingMultiplier=~
    ParachutingAdditive=~
   )
   n4 = Get gravity_override() -> (gravity_override:Vector)
   n5 = RigidBody!(ComponentPose=@n2, Alpha, OverrideWorldGravity:Vector=@n4, ExternalForce=@n3.SkydivingVectorForce) -> (Pose) {
    bOverrideWorldGravity=true
    overridePhysicsAsset=&/F_MED_Swashbuckler_Dyn_Skirt_Back
    simulationSpace=BaseBoneSpace
    alphaBoolBlend {blendOption=Linear}
    baseBoneRef {boneName=spine_01}
    componentAppliedLinearAccClamp=<3000 3000 3000>
    componentLinearAccScale=<0.4 0.2 0.2>
    componentLinearVelScale=<0.4 0.2 0.2>
    overrideWorldGravity {z=-980}
   }
   n6 = RigidBody(ComponentPose=@n5, OverrideWorldGravity=@n4, ExternalForce=@n3.SkydivingVectorForce) {
    bOverrideWorldGravity=true
    overridePhysicsAsset=&/F_MED_Swashbuckler_Dyn_Skirt_Front
    simulationSpace=BaseBoneSpace
    alphaBoolBlend {blendOption=Linear}
    baseBoneRef {boneName=thigh_r_blend}
    componentAppliedLinearAccClamp=<3000 3000 3000>
    componentLinearAccScale=<0.3 0.3 0.3>
    componentLinearVelScale=<0.3 0.3 0.3>
    overrideWorldGravity {z=-980}
   }
   n7 = Anim Dynamics - Bone: dyn_wrist_flap_l(ComponentPose=@n6, Alpha=1, ExternalForce="0,0,0") {
    angularDampingOverride=1
    angularSpringConstant=12000
    bAngularSpring=true
    bGravityOverrideInSimSpace=true
    bOverrideAngularDamping=true
    bUseGravityOverride=true
    gravityScale=0.35
    simulationSpace=RootRelative
    alphaBoolBlend {blendOption=Linear}
    boundBone {boneName=dyn_wrist_flap_l}
    gravityOverride {z=-980}
    physicsBodyDefinitions [{
      sphereCollisionRadius=0
      boundBone {boneName=dyn_wrist_flap_l}
      boxExtents=<10 20 20>
      constraintSetup {angularLimitsMax=<0 5 5>, angularLimitsMin=<0 -5 -5>, angularTarget=<1 0 0>}
      localJointOffset=<0.2 0 0>
     }]
   }
   n8 = Anim Dynamics - Bone: dyn_wrist_flap_r(ComponentPose=@n7, Alpha=1, ExternalForce="0,0,0") {
    angularDampingOverride=1
    angularSpringConstant=12000
    bAngularSpring=true
    bGravityOverrideInSimSpace=true
    bOverrideAngularDamping=true
    bUseGravityOverride=true
    gravityScale=0.35
    simulationSpace=RootRelative
    alphaBoolBlend {blendOption=Linear}
    boundBone {boneName=dyn_wrist_flap_r}
    gravityOverride {z=-980}
    physicsBodyDefinitions [{
      sphereCollisionRadius=0
      boundBone {boneName=dyn_wrist_flap_r}
      boxExtents=<10 20 20>
      constraintSetup {angularLimitsMax=<0 5 5>, angularLimitsMin=<0 -5 -5>, angularTarget=<1 0 0>}
      localJointOffset=<-0.2 0 0>
     }]
   }
   n9 = Anim Dynamics - Bone: dyn_calf_boot_cuff_l(ComponentPose=@n8, Alpha=1, ExternalForce="0,0,0") {
    angularDampingOverride=1
    angularSpringConstant=8000
    bAngularSpring=true
    bGravityOverrideInSimSpace=true
    bOverrideAngularDamping=true
    bUseGravityOverride=true
    gravityScale=0.35
    simulationSpace=RootRelative
    alphaBoolBlend {blendOption=Linear}
    boundBone {boneName=dyn_calf_boot_cuff_l}
    gravityOverride {z=-980}
    physicsBodyDefinitions [{
      sphereCollisionRadius=0
      boundBone {boneName=dyn_calf_boot_cuff_l}
      boxExtents=<10 20 20>
      constraintSetup {angularLimitsMax=<0 15 15>, angularLimitsMin=<0 -15 -15>, angularTarget=<1 0 0>}
      localJointOffset=<2 0 0>
     }]
   }
   n10 = Anim Dynamics - Bone: dyn_calf_boot_cuff_r(ComponentPose=@n9, Alpha=1, ExternalForce="0,0,0") {
    angularDampingOverride=1
    angularSpringConstant=8000
    bAngularSpring=true
    bGravityOverrideInSimSpace=true
    bOverrideAngularDamping=true
    bUseGravityOverride=true
    gravityScale=0.35
    simulationSpace=RootRelative
    alphaBoolBlend {blendOption=Linear}
    boundBone {boneName=dyn_calf_boot_cuff_r}
    gravityOverride {z=-980}
    physicsBodyDefinitions [{
      sphereCollisionRadius=0
      boundBone {boneName=dyn_calf_boot_cuff_r}
      boxExtents=<10 20 20>
      constraintSetup {angularLimitsMax=<0 15 15>, angularLimitsMin=<0 -15 -15>, angularTarget=<1 0 0>}
      localJointOffset=<-2 0 0>
     }]
   }
   n11 = "Constraint - thigh_r_blend (R)"(ComponentPose=Local To Component(LocalPose=Use cached pose 'copypose_cache'()), Alpha=0.5, ConstraintWeights_0=1) {
    alphaBoolBlend {blendOption=Linear}
    boneToModify {boneName=thigh_r_blend}
    constraintSetup [{transformType=Rotation, targetBone {boneName=thigh_r}}]
    constraintWeights [1]
   }
   n12 = "Constraint - thigh_l_blend (R)"(ComponentPose=@n11, Alpha=0.5, ConstraintWeights_0=1) {
    alphaBoolBlend {blendOption=Linear}
    boneToModify {boneName=thigh_l_blend}
    constraintSetup [{transformType=Rotation, targetBone {boneName=thigh_l}}]
    constraintWeights [1]
   }
   n13 = RigidBody(ComponentPose=@n12, OverrideWorldGravity=@n4, ExternalForce=@n3.SkydivingVectorForce) {
    bOverrideWorldGravity=true
    overridePhysicsAsset=&/F_MED_Swashbuckler_Dyn_Skirt_Back_Frontend
    simulationSpace=BaseBoneSpace
    alphaBoolBlend {blendOption=Linear}
    baseBoneRef {boneName=spine_01}
    componentAppliedLinearAccClamp=<3000 3000 3000>
    componentLinearAccScale=<0.4 0.2 0.2>
    componentLinearVelScale=<0.4 0.2 0.2>
    overrideWorldGravity {z=-980}
   }
   n14 = RigidBody(ComponentPose=@n13, OverrideWorldGravity=@n4, ExternalForce=@n3.SkydivingVectorForce) {
    bOverrideWorldGravity=true
    overridePhysicsAsset=&F_MED_Swashbuckler_Dyn_Skirt_Front
    simulationSpace=BaseBoneSpace
    alphaBoolBlend {blendOption=Linear}
    baseBoneRef {boneName=thigh_r_blend}
    componentAppliedLinearAccClamp=<3000 3000 3000>
    componentLinearAccScale=<0.3 0.3 0.3>
    componentLinearVelScale=<0.3 0.3 0.3>
    overrideWorldGravity {z=-980}
   }
   n15 = Anim Dynamics - Bone: dyn_wrist_flap_l(ComponentPose=@n14, Alpha=1, ExternalForce="0,0,0") {
    angularDampingOverride=1
    angularSpringConstant=12000
    bAngularSpring=true
    bGravityOverrideInSimSpace=true
    bOverrideAngularDamping=true
    bUseGravityOverride=true
    gravityScale=0.35
    simulationSpace=RootRelative
    alphaBoolBlend {blendOption=Linear}
    boundBone {boneName=dyn_wrist_flap_l}
    gravityOverride {z=-980}
    physicsBodyDefinitions [{
      sphereCollisionRadius=0
      boundBone {boneName=dyn_wrist_flap_l}
      boxExtents=<10 20 20>
      constraintSetup {angularLimitsMax=<0 5 5>, angularLimitsMin=<0 -5 -5>, angularTarget=<1 0 0>}
      localJointOffset=<0.2 0 0>
     }]
   }
   n16 = Anim Dynamics - Bone: dyn_wrist_flap_r(ComponentPose=@n15, Alpha=1, ExternalForce="0,0,0") {
    angularDampingOverride=1
    angularSpringConstant=12000
    bAngularSpring=true
    bGravityOverrideInSimSpace=true
    bOverrideAngularDamping=true
    bUseGravityOverride=true
    gravityScale=0.35
    simulationSpace=RootRelative
    alphaBoolBlend {blendOption=Linear}
    boundBone {boneName=dyn_wrist_flap_r}
    gravityOverride {z=-980}
    physicsBodyDefinitions [{
      sphereCollisionRadius=0
      boundBone {boneName=dyn_wrist_flap_r}
      boxExtents=<10 20 20>
      constraintSetup {angularLimitsMax=<0 5 5>, angularLimitsMin=<0 -5 -5>, angularTarget=<1 0 0>}
      localJointOffset=<-0.2 0 0>
     }]
   }
   n17 = Anim Dynamics - Bone: dyn_calf_boot_cuff_l(ComponentPose=@n16, Alpha=1, ExternalForce="0,0,0") {
    angularDampingOverride=1
    angularSpringConstant=8000
    bAngularSpring=true
    bGravityOverrideInSimSpace=true
    bOverrideAngularDamping=true
    bUseGravityOverride=true
    gravityScale=0.35
    simulationSpace=RootRelative
    alphaBoolBlend {blendOption=Linear}
    boundBone {boneName=dyn_calf_boot_cuff_l}
    gravityOverride {z=-980}
    physicsBodyDefinitions [{
      sphereCollisionRadius=0
      boundBone {boneName=dyn_calf_boot_cuff_l}
      boxExtents=<10 20 20>
      constraintSetup {angularLimitsMax=<0 15 15>, angularLimitsMin=<0 -15 -15>, angularTarget=<1 0 0>}
      localJointOffset=<2 0 0>
     }]
   }
   n18 = Anim Dynamics - Bone: dyn_calf_boot_cuff_r(ComponentPose=@n17, Alpha=1, ExternalForce="0,0,0") {
    angularDampingOverride=1
    angularSpringConstant=8000
    bAngularSpring=true
    bGravityOverrideInSimSpace=true
    bOverrideAngularDamping=true
    bUseGravityOverride=true
    gravityScale=0.35
    simulationSpace=RootRelative
    alphaBoolBlend {blendOption=Linear}
    boundBone {boneName=dyn_calf_boot_cuff_r}
    gravityOverride {z=-980}
    physicsBodyDefinitions [{
      sphereCollisionRadius=0
      boundBone {boneName=dyn_calf_boot_cuff_r}
      boxExtents=<10 20 20>
      constraintSetup {angularLimitsMax=<0 15 15>, angularLimitsMin=<0 -15 -15>, angularTarget=<1 0 0>}
      localJointOffset=<-2 0 0>
     }]
   }
   n19 = Blend Poses by bool(
    BlendPose_0=Component To Local(ComponentPose=@n18)
    BlendPose_1=Component To Local(ComponentPose=@n10)
    bActiveValue=Get bCanPlayCustomAnimations() -> (bCanPlayCustomAnimations:bool)
   ) {blendType=Linear}
   n20 = "Constraint - dyn_tail_l_01 (P)"(ComponentPose=Local To Component(LocalPose=Use cached pose 'copypose_cache'()), Alpha=1, ConstraintWeights_0=1) {
    alphaBoolBlend {blendOption=Linear}
    boneToModify {boneName=dyn_tail_l_01}
    constraintSetup [{targetBone {boneName=spine_01}}]
    constraintWeights [1]
   }
   n21 = "Transform (Modify) Bone - Bone: dyn_tail_l_01"(ComponentPose=@n20, Alpha=1, Translation="0.000,0.000,1.779", Rotation="0,0,0", Scale="1,1,1") {
    translationMode=BMM_Additive
    translationSpace=BCS_BoneSpace
    alphaBoolBlend {blendOption=Linear}
    boneToModify {boneName=dyn_tail_l_01}
    translation {z=1.779}
   }
   n22 = "Constraint - dyn_tail_mid_l_01 (P)"(ComponentPose=@n21, Alpha=1, ConstraintWeights_0=1) {
    alphaBoolBlend {blendOption=Linear}
    boneToModify {boneName=dyn_tail_mid_l_01}
    constraintSetup [{targetBone {boneName=spine_01}}]
    constraintWeights [1]
   }
   n23 = "Transform (Modify) Bone - Bone: dyn_tail_mid_l_01"(ComponentPose=@n22, Alpha=1, Translation="-0.000,0.000,1.994", Rotation="0,0,0", Scale="1,1,1") {
    translationMode=BMM_Additive
    translationSpace=BCS_BoneSpace
    alphaBoolBlend {blendOption=Linear}
    boneToModify {boneName=dyn_tail_mid_l_01}
    translation {z=1.994}
   }
   n24 = "Constraint - dyn_tail_mid_r_01 (P)"(ComponentPose=@n23, Alpha=1, ConstraintWeights_0=1) {
    alphaBoolBlend {blendOption=Linear}
    boneToModify {boneName=dyn_tail_mid_r_01}
    constraintSetup [{targetBone {boneName=spine_01}}]
    constraintWeights [1]
   }
   n25 = "Constraint - dyn_tail_r_01 (P)"(ComponentPose=@n24, Alpha=1, ConstraintWeights_0=1) {
    alphaBoolBlend {blendOption=Linear}
    boneToModify {boneName=dyn_tail_r_01}
    constraintSetup [{targetBone {boneName=spine_01}}]
    constraintWeights [1]
   }
   n26 = "Constraint - dyn_tail_r_02 (PP)"!(ComponentPose=@n25, Alpha=1, ConstraintWeights_0=0.8927, ConstraintWeights_1:float=0.1073) -> (Pose) {
    alphaBoolBlend {blendOption=Linear}
    boneToModify {boneName=dyn_tail_r_02}
    constraintSetup [{targetBone {boneName=pelvis}}, {targetBone {boneName=thigh_twist_01_r}}]
    constraintWeights [1,1]
   }
   n27 = "Constraint - dyn_tail_mid_r_02 (PP)"!(ComponentPose=@n26, Alpha=1, ConstraintWeights_0=0.9703, ConstraintWeights_1:float=0.0297) -> (Pose) {
    alphaBoolBlend {blendOption=Linear}
    boneToModify {boneName=dyn_tail_mid_r_02}
    constraintSetup [{targetBone {boneName=pelvis}}, {targetBone {boneName=thigh_twist_01_r}}]
    constraintWeights [1,1]
   }
   n28 = "Constraint - dyn_tail_mid_l_02 (PP)"!(ComponentPose=@n27, Alpha=1, ConstraintWeights_0=0.9489, ConstraintWeights_1:float=0.0511) -> (Pose) {
    alphaBoolBlend {blendOption=Linear}
    boneToModify {boneName=dyn_tail_mid_l_02}
    constraintSetup [{targetBone {boneName=pelvis}}, {targetBone {boneName=thigh_twist_01_l}}]
    constraintWeights [1,1]
   }
   n29 = "Constraint - dyn_tail_l_02 (PP)"!(ComponentPose=@n28, Alpha=1, ConstraintWeights_0=0.8927, ConstraintWeights_1:float=0.1073) -> (Pose) {
    alphaBoolBlend {blendOption=Linear}
    boneToModify {boneName=dyn_tail_l_02}
    constraintSetup [{targetBone {boneName=pelvis}}, {targetBone {boneName=thigh_twist_01_l}}]
    constraintWeights [1,1]
   }
   n30 = "Constraint - dyn_tail_r_03 (PPP)"!(ComponentPose=@n29, Alpha=1, ConstraintWeights_0=0.4, ConstraintWeights_1:float=0.6, ConstraintWeights_2:float=0) -> (Pose) {
    alphaBoolBlend {blendOption=Linear}
    boneToModify {boneName=dyn_tail_r_03}
    constraintSetup [{targetBone {boneName=thigh_r}}, {targetBone {boneName=pelvis}}, {targetBone {boneName=thigh_twist_01_r}}]
    constraintWeights [1,1,1]
   }
   n31 = "Constraint - dyn_tail_mid_r_03 (PPP)"!(ComponentPose=@n30, Alpha=1, ConstraintWeights_0=0.3, ConstraintWeights_1:float=0.7, ConstraintWeights_2:float=0) -> (Pose) {
    alphaBoolBlend {blendOption=Linear}
    boneToModify {boneName=dyn_tail_mid_r_03}
    constraintSetup [{targetBone {boneName=thigh_r}}, {targetBone {boneName=pelvis}}, {targetBone {boneName=thigh_twist_01_r}}]
    constraintWeights [1,1,1]
   }
   n32 = "Constraint - dyn_tail_mid_l_03 (PPP)"!(ComponentPose=@n31, Alpha=1, ConstraintWeights_0=0.678, ConstraintWeights_1:float=0, ConstraintWeights_2:float=0.3) -> (Pose) {
    alphaBoolBlend {blendOption=Linear}
    boneToModify {boneName=dyn_tail_mid_l_03}
    constraintSetup [{targetBone {boneName=pelvis}}, {targetBone {boneName=thigh_twist_01_l}}, {targetBone {boneName=thigh_l}}]
    constraintWeights [1,1,1]
   }
   n33 = "Constraint - dyn_tail_l_03 (PPP)"!(ComponentPose=@n32, Alpha=1, ConstraintWeights_0=0.35, ConstraintWeights_1:float=0, ConstraintWeights_2:float=0.65) -> (Pose) {
    alphaBoolBlend {blendOption=Linear}
    boneToModify {boneName=dyn_tail_l_03}
    constraintSetup [{targetBone {boneName=pelvis}}, {targetBone {boneName=thigh_twist_01_l}}, {targetBone {boneName=thigh_l}}]
    constraintWeights [1,1,1]
   }
   n34 = Component To Local(ComponentPose=@n33)
   n35 = Get rigidbody_disabled() -> (rigidbody_disabled:bool)
   n36 = Get Is Sky Diving() -> (Is Sky Diving:double)
   n37 = "Transform (Modify) Bone - Bone: dyn_tail_mid_r_01"(
    ComponentPose=Local To Component(LocalPose=Blend Poses by bool(BlendPose_0=@n34, BlendPose_1=@n19, bActiveValue=@n35) {blendType=Linear})
    Alpha=@n36
    Translation="0,0,0"
    Rotation="10,0,0"
    Scale="1,1,1"
   ) {
    rotationMode=BMM_Additive
    rotationSpace=BCS_BoneSpace
    translationMode=BMM_Additive
    translationSpace=BCS_BoneSpace
    alphaBoolBlend {blendOption=Linear}
    alphaScaleBiasClamp {bInterpResult=true}
    boneToModify {boneName=dyn_tail_mid_r_01}
    rotation {pitch=10}
   }
   n38 = "Transform (Modify) Bone - Bone: dyn_tail_r_01"(ComponentPose=@n37, Alpha=@n36, Translation="0,0,0", Rotation="10,0,0", Scale="1,1,1") {
    rotationMode=BMM_Additive
    rotationSpace=BCS_BoneSpace
    translationMode=BMM_Additive
    translationSpace=BCS_BoneSpace
    alphaBoolBlend {blendOption=Linear}
    alphaScaleBiasClamp {bInterpResult=true}
    boneToModify {boneName=dyn_tail_r_01}
    rotation {pitch=10}
   }
   n39 = "Transform (Modify) Bone - Bone: dyn_tail_mid_l_01"(ComponentPose=@n38, Alpha=@n36, Translation="0,0,0", Rotation="10,0,0", Scale="1,1,1") {
    rotationMode=BMM_Additive
    rotationSpace=BCS_BoneSpace
    translationMode=BMM_Additive
    translationSpace=BCS_BoneSpace
    alphaBoolBlend {blendOption=Linear}
    alphaScaleBiasClamp {bInterpResult=true}
    boneToModify {boneName=dyn_tail_mid_l_01}
    rotation {pitch=10}
   }
   n40 = "Transform (Modify) Bone - Bone: dyn_tail_l_01"(ComponentPose=@n39, Alpha=@n36, Translation="0,0,0", Rotation="10,0,0", Scale="1,1,1") {
    rotationMode=BMM_Additive
    rotationSpace=BCS_BoneSpace
    translationMode=BMM_Additive
    translationSpace=BCS_BoneSpace
    alphaBoolBlend {blendOption=Linear}
    alphaScaleBiasClamp {bInterpResult=true}
    boneToModify {boneName=dyn_tail_l_01}
    rotation {pitch=10}
   }
   Output Pose(Result=Component To Local(ComponentPose=@n40)) {layerGroup=None}
   Save cached pose 'copypose_cache'(Pose=Copy Pose From Mesh() {bUseAttachedParent=true}) {cacheName=copypose_cache}
  ]}
 EventGraph {nodes [
   Event BlueprintUpdateAnimation()
   > n1 = Sequence!(execute) -> (then_0, then_1, then_2:Exec)
   n2 = Get bRigidBodyEnabled()
   n3 = Get CurrentLOD()
   > Set rigidbody_disabled(rigidbody_disabled:bool=OR Boolean(A=NOT Boolean(A=@n2), B="integer > integer"(A=@n3, B=1) {is=title K2Node_CallFunction})) -> (Output_Get:bool)
   n4 = Branch(execute=@n1.then_1, Condition=Get bIsSwimSprinting() -> (bIsSwimSprinting:bool))
   > Set gravity_override(gravity_override="0, 0, 0")
   n5 = Branch(execute=@n4.else, Condition=Get bIsSkyDiving())
   > n6 = Branch(Condition=Get bIsParachuteOpened())
   Set gravity_override(execute=[@n5.else, @n6.then], gravity_override="0, 0,-980")
   n7 = Branch(execute=@n6.else, Condition="float > float"(A=Get PawnVelocityZ(), B=50) {is=title K2Node_CallFunction})
   > Set gravity_override(gravity_override="0, 0,200")
   n8 = Get Skel_Mesh()
   n9 = vector * float(
    A=GetForwardVector!(InRot:Rotator=GetSocketRotation!(self:Scene Component Object Reference=@n8, InSocketName:Name=pelvis) -> (ReturnValue:Rotator)) -> (ReturnValue:Vector)
    B=-1000
   )
   Set gravity_override(
    execute=@n7.else
    gravity_override=vector + vector(A:Vector=@n9, B:Vector="0, 0,-200") -> (ReturnValue:Vector) {is=title K2Node_CommutativeAssociativeBinaryOperator}
   )
   Set Is Sky Diving(execute=@n1.then_2, Is Sky Diving:double="To Float (Boolean)"(InBool=Get bIsSkyDiving())) -> (Output_Get:double)
   Event BlueprintInitializeAnimation()
   > Set Skel_Mesh(Skel_Mesh=GetOwnerSkeletalMeshComponent())
   n10 = GetSocketLocation!(self:Scene Component Object Reference=@n8, InSocketName:Name=pelvis) -> (ReturnValue:Vector)
   DrawDebugLine!(
    execute
    LineStart:Vector=@n10
    LineEnd:Vector=vector + vector(A=@n10, B=@n9)
    LineColor:Linear Color Structure=(R=1,G=0,B=0,A=1)
    Duration:float=0
    Thickness:float=3
    DepthPriority:EDrawDebugSceneDepthPriorityGroup Enum=World
   ) -> (then)
  ]}
]
variables [
 gravity_override {type=Vector}
 Is Sky Diving {type=number}
 rigidbody_disabled {type=boolean}
 Skel_Mesh {type=SkeletalMeshComponent}
]
